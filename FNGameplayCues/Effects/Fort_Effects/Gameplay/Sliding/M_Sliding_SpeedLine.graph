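# Translucent, unlit, two-sided material for Niagara mesh particles that draws animated speed-line streaks for the sliding effect.
# Two stretched samples of a pixel texture pan at different speeds, and their difference is masked by an axis-swapped lightning texture.
# Particle color tints the emissive, while particle alpha and a powered sine along U shape the opacity.
# A scalar parameter blends in custom UV scaling for all three samples.
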Material /FNGameplayCues/Effects/Fort_Effects/Gameplay/Sliding/M_Sliding_SpeedLine
graph {
 properties {
  bUsedWithNiagaraMeshParticles=true
  blendMode=BLEND_Translucent
  pixelDepthOffsetMode=PDOM_Legacy
  shadingModel=MSM_Unlit
  twoSided=true
  usedShadingModels=Unlit
 }
 nodes [
  n1 = TextureCoordinate() {uTiling=0.1, vTiling=22}
  n2 = VectorParameter() {parameterName=UV_Scale, channelNames {a="", b="", g=VScale, r=UScale}, defaultValue=<rgba 1 1 0 0>}
  n3 = Multiply(A=@n1, B=ComponentMask(None=@n2) {g=true, r=true})
  n4 = ScalarParameter() {parameterName=Use Custom UVScale}
  n5 = TextureSample!(UVs=Panner(Coordinate=LinearInterpolate(A=@n1, B=@n3, Alpha=@n4)) {speedX=0.65, speedY=0.1}, Tex, Level, Apply View MipBias) {
   mipValueMode=TMVM_MipLevel
   samplerType=SAMPLERTYPE_LinearGrayscale
   texture=/Game/UI/Frontend/ItemShop/Com/vfx/T_CommunityVoting_Activities_Pixels.CommunityVoting_Activities_Pixels
  }
  n6 = TextureCoordinate() {uTiling=0.1, vTiling=14}
  n7 = VectorParameter() {parameterName=UV_Scale, channelNames {a="", b="", g=VScale, r=UScale}, defaultValue=<rgba 1 1 0 0>}
  n8 = Multiply(A=@n6, B=ComponentMask(None=@n7) {g=true, r=true})
  n9 = ScalarParameter() {parameterName=Use Custom UVScale}
  n10 = TextureSample(UVs=Panner(Coordinate=LinearInterpolate(A=@n6, B=@n8, Alpha=@n9)) {speedX=0.7, speedY=0.3}) {
   mipValueMode=TMVM_MipLevel
   samplerType=SAMPLERTYPE_LinearGrayscale
   texture=&CommunityVoting_Activities_Pixels
  }
  n11 = TextureCoordinate() {vTiling=5}
  n12 = VectorParameter() {parameterName=UV_Scale, channelNames {a="", b="", g=VScale, r=UScale}, defaultValue=<rgba 1 1 0 0>}
  n13 = ComponentMask(None=@n12) {g=true, r=true}
  n14 = MaterialFunctionCall(XY=LinearInterpolate(A=@n11, B=Multiply(A=@n11, B=@n13), Alpha=ScalarParameter() {parameterName=Use Custom UVScale})) {materialFunction=&Swizzle}
  n15 = Abs(None=Multiply(
    A=Subtract(A=@n5.R, B=@n10.R)
    B=TextureSample(UVs=@n14.YX) {mipValueMode=TMVM_MipLevel, texture=/Game/VisualThreatManager/StormVisuals/Textures/StormLightning}.R
   ))
  n16 = ParticleColor()
  MaterialFunctionCall(Emissive=Multiply(A=@n15, B=@n16.RGB)) {materialFunction=&MF_FX_EmissiveAdjustments}
  n17 = TextureCoordinate()
  n18 = Sine(None=Divide(A=MaterialFunctionCall(Float2=@n17) {materialFunction=&BreakOutFloat2Components}.R))
  Multiply(A=@n16.A, B=Multiply(A=@n15, B=Power(Base=@n18)))
 ]
}
